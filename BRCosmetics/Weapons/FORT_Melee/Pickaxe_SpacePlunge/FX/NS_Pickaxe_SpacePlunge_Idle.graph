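# Niagara idle effect for the SpacePlunge pickaxe, with looks selected by User.Variant.
# Arks flickers short-lived arc sprites and lights on weapon sockets, feeding Smoke and Sparks through particle reads, all switching between orange and cyan palettes.
# IdleCubes001 spawns only for variant 2, shedding tumbling, shrinking meshes from a socket while blending between local and world space.
# Sparks thin out through ReduceSpawnWhileRunning, and the older IdleCubes emitter is disabled.
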NiagaraSystem /BRCosmetics/Weapons/FORT_Melee/Pickaxe_SpacePlunge/FX/NS_Pickaxe_SpacePlunge_Idle
dependencies {
 usedDataInterfaces [
  &NiagaraDataInterfaceArrayColor
  &NiagaraDataInterfaceArrayFloat
  &NiagaraDataInterfaceArrayFloat3
  &NiagaraDataInterfaceCurve
  &NiagaraDataInterfaceArrayInt32
  &NiagaraDataInterfaceSpriteRendererInfo
  &NiagaraDataInterfaceParticleRead
  &NiagaraDataInterfaceMeshRendererInfo
  &NiagaraDataInterfaceVectorCurve
  &NiagaraDataInterfaceSkeletalMesh
  &NiagaraDataInterfaceColorCurve
 ]
 usedDynamicInputs [&Add_Float, &MakeVector]
 usedModules [
  &self:SystemSpawnScript.SystemScriptSource.SystemScriptGraph.NiagaraNodeAssignment_0.SetVariables_#1
  &self:SystemSpawnScript.SystemScriptSource.SystemScriptGraph.NiagaraNodeAssignment_1.SetVariables_#2
  &self:SystemSpawnScript.SystemScriptSource.SystemScriptGraph.NiagaraNodeAssignment_2.SetVariables_#3
  &SystemState
  &BlendSpaceHelper
 ]
}
emitters {
 Arks {
  emitterUpdateScript [
   &EmitterState(Inactive Response=::NewEnumerator1, Life Cycle Mode=::NewEnumerator1, Loop Duration=1, UseLoopDelay=false)
   /SetVariables(Emitter.SpawnCount:NiagaraInt32=&RandomRangeInt(Minimum=0, Maximum=4)) {isSetParameters=true}
   &SpawnRate(
    Spawn Probability=&RandomRangeFloat(Minimum=0.1, Maximum=0.4)
    SpawnRate=&RandomRangeFloat(Minimum=1, Maximum=3)
    Use Spawn Probability=true
   )
  ]
  particleSpawnScript [
   &PartitionParticles(Partition Mode:ENiagara_PartitionMode_v2=::NewEnumerator5)
   &InitializeParticle(
    Color=&MaskLinearColor(
     Mask Source=1
     Mask Parameter=&MakeIntFromFloat(Float=$User.Variant, Conversion Mode=::NewEnumerator1)
     Mask=1
     Masked Linear Color=&SelectLinearColorFromArray(
      Array Sampling Mode=::NewEnumerator0
      Direct Array Index=?
      Color Selection Array=DataInterface() {propertyValues {ColorData [<rgba 0 0.9426 10 1>, <rgba 0 5.5332 10 1>, <rgba 0 10 8.1713 1>]}}
      Random Seed=0
     )
     Passthrough Linear Color=&SelectLinearColorFromArray(
      Array Sampling Mode=::NewEnumerator0
      Direct Array Index=?
      Color Selection Array=DataInterface() {propertyValues {ColorData [<rgba 10 5.6687 0.1522 1>, <rgba 10 3.3 0.287 1>, <rgba 10 2.5941 0.077 1>]}}
      Random Seed=0
     )
    )
    Lifetime Max=0.35
    Lifetime Min=0.25
    Lifetime Mode=::NewEnumerator1
    Sprite Size Max=&MakeVector2D(
     X=&Multiply_Float(
      A=8
      B=&UniformAOrBFloat(A=-1, Evaluation Type=::NewEnumerator0, Recalculate Random Each Loop=true, Override Seed=false)
     )
     Y=8
    )
    Sprite Size Min=&MakeVector2D(
     X=&Multiply_Float(
      A=4
      B=&UniformAOrBFloat(A=-1, Evaluation Type=::NewEnumerator0, Recalculate Random Each Loop=true, Override Seed=false)
     )
     Y=4
    )
    Sprite Size Mode=::NewEnumerator2
    UsePositionOffset=false
    Write Material Random=true
   )
   /SetVariables(Particles.FlareSize:Vector2f=&Multiply_Vector2D_ByFloat(Float=&RandomRangeFloat(Minimum=0.75, Maximum=1), Vector2D=<20 20>)) {isSetParameters=true}
   /SetVariables(Particles.FlareCameraOffset=20) {isSetParameters=true}
   /SetVariables(Particles.FlareColor=&MaskLinearColor(
     Mask Source=1
     Mask Parameter=&MakeIntFromFloat(Float=$User.Variant, Conversion Mode=::NewEnumerator1)
     Mask=1
     Masked Linear Color=<rgba 0 10 10 0.302>
     Passthrough Linear Color=<rgba 10 4.0357 0.1929 0.302>
    )) {isSetParameters=true}
  ]
  particleUpdateScript [
   &ParticleState()
   /SetVariables(Particles.SpriteAlignment=&SelectVectorFromArray(Direct Array Index=0, Vector Selection Array=DataInterface() {propertyValues {FloatData [<-1 0 0>]}})) {isSetParameters=true}
   &ScaleSpriteSize(Uniform Curve Sprite Scale=DataInterface() {propertyValues {Curve {keys [{interpMode=RCIM_Cubic, tangentMode=RCTM_User, time=0, value=0}, {interpMode=RCIM_Cubic, tangentMode=RCTM_User, time=1, value=1}]}}})
   &SkeletalMeshLocation(
    Bone / Socket Index=&SelectIntFromArray(
     Direct Array Index=&RandomRangeInt(Minimum=0, Maximum=1)
     Int Selection Array=DataInterface() {propertyValues {IntData [0,1]}}
    )
    MaskSpawnGroup=false
    Mesh Sampling Type=::NewEnumerator5
    Sampled Position Offset Coordinate Space:ENiagaraSkelMeshTransforms=::NewEnumerator1
    Skeletal Mesh=$System.Skeletal Mesh
    Socket Sampling Mode=::NewEnumerator5
   )
   &SubUVAnimation(
    End Frame Range Override=3
    "Sprite Renderer "=DataInterface() {propertyValues {SpriteRenderer=&self:Arks_1.NiagaraSpriteRendererProperties_0}}
    SubUV Animation Mode:ENiagara_SubUVLookupModeV2=::NewEnumerator2
    UseEndFrame=true
    UseStartFrame=false
   )
   &DynamicMaterialParameters(
    Index 0 Param 2=10
    Module.Index0Float0Write=true
    Module.Index0Float1Write=true
    Module.Index0Float2Write=true
    Module.Index0Float3Write=true
    Module.Index1Float0Write=true
    Module.Index1Float1Write=true
    Module.Index1Float2Write=true
    Module.Index1Float3Write=true
    Write Parameter Index 1=::NewEnumerator1
   ) {module=DynamicMaterialParameters001}
   /SetVariables(
    Particles.LightColor=&MultiplyLinearColor(
     Linear Color=$Particles.Color
     Scale Factor=&MakeVector4(
      X=1
      Y=1
      Z=1
      W=&Multiply_Float(A=1, B=&MakeFloatFromVector(Vector=$Engine.Owner.Scale, Channel=::NewEnumerator2))
     )
    )
    Particles.LightRadius=5
   ) {isSetParameters=true}
  ]
  rendererClasses [&NiagaraSpriteRendererProperties, &NiagaraLightRendererProperties]
  renderers [
   {rendererClass=&NiagaraSpriteRendererProperties}
   {rendererClass=&NiagaraLightRendererProperties, rendererIndex=1}
   {rendererClass=&NiagaraSpriteRendererProperties, rendererIndex=2}
  ]
 }
 IdleCubes {
  bEnabled=false
  emitterUpdateScript [
   &EmitterState()
   &SpawnBurst_Instantaneous(
    Spawn Count=15
    Spawn Group=1
    Spawn Probability=&MaskFloat(
     Mask Source=1
     Mask Parameter=&MakeIntFromFloat(Float=$User.Variant, Conversion Mode=::NewEnumerator1)
     Mask=2
     Masked Float=1
     Passthrough Float=0
    )
    Use Spawn Probability=true
    UseLoopCountLimit=true
   )
   &SpawnRate(
    Spawn Probability=&MaskFloat(
     Mask Source=1
     Mask Parameter=&MakeIntFromFloat(Float=$User.Variant, Conversion Mode=::NewEnumerator1)
     Mask=2
     Masked Float=1
     Passthrough Float=0
    )
    SpawnRate=5
    Use Spawn Probability=true
   )
  ]
  particleSpawnScript [
   &PartitionParticles()
   &InitializeParticle(
    Color Maximum=<rgba 1 0.3864 0.0138 1>
    Color Minimum=<rgba 0.6094 0.6094 0.6094 1>
    Color Mode=::NewEnumerator2
    Lifetime=&RandomRangeFloat(Minimum=6, Maximum=10)
    Mesh Scale Mode=::NewEnumerator3
    UsePositionOffset=false
    Write Material Random=true
   )
   /SetVariables() {enabled=false, isSetParameters=true}
   &SubUVAnimation(
    Mesh Renderer=DataInterface() {propertyValues {MeshRenderer=&self:IdleCubes_0.NiagaraMeshRendererProperties_0}}
    Output Alternate Renderer Binding=true
    Random Start Frame=true
    Renderer Type:ENiagaraMeshOrSprite=::NewEnumerator1
    SubUV Lookup Index Scale=0
    UseEndFrame=false
    UseStartFrame=false
   )
   &InitialMeshOrientation(Mesh Orientation Mode=::NewEnumerator7, Use Rotation Vector=false) {enabled=false}
  ]
  particleUpdateScript [
   &ParticleState()
   &ScaleMeshSize(Scale Factor=&VectorFromCurve(
     VectorCurve=DataInterface() {propertyValues {
       XCurve {keys [<at -0.0042 0.9902>,<at 0.3384 0.9047>,<at 0.9927 0.948>]}
       YCurve {keys [<at -0.0042 0.9902>,<at 0.4836 0.8148>,<at 0.9823 0.7316>]}
       ZCurve {keys [<at 0 0.577>,<at 0.2083 1.0993>,<at 0.6031 1.0477>,<at 1.0031 0.7906>]}
      }}
     Scale Curve=&Multiply_VectorByFloat(
      Vector=<1 1 1>
      Float=&FloatFromCurve(FloatCurve=DataInterface() {propertyValues {Curve {keys [
           {interpMode=RCIM_Cubic, tangentMode=RCTM_User, time=0, value=0.25}
           {interpMode=RCIM_Cubic, tangentMode=RCTM_User, time=0.2909, value=1}
           {interpMode=RCIM_Cubic, tangentMode=RCTM_User, time=0.8, value=1}
           {arriveTangent=-4.1587, interpMode=RCIM_Cubic, leaveTangent=-4.1587, tangentMode=RCTM_User, time=1, value=-0.0118}
          ]}}})
     )
    )) {enabled=false, module=ScaleMeshSize002}
   &SkeletalMeshLocation(
    Bone / Socket Index=&SelectIntFromArray(Direct Array Index=$Particles.PartitionParticles.Partition, Int Selection Array=DataInterface() {propertyValues {IntData [4,5]}})
    MaskSpawnGroup=false
    Mesh Sampling Type=::NewEnumerator5
    Sampled Orientation Coordinate Space=::Local
    Sampled Position Offset=&Lerp_Vector(
     A=&RandomRangeVector(Minimum=<15 -1 -4>, Maximum=<30 1 5>)
     B=&RandomRangeVector(Minimum=<-3 -1 0>, Maximum=<3 1 20>)
     Alpha=&MakeFloatFromInt(Int=$Particles.PartitionParticles.Partition)
    )
    Skeletal Mesh=$System.Skeletal Mesh
    Socket Sampling Mode=::NewEnumerator5
   )
   &SkeletalMeshLocation(
    Bone / Socket Index=&SelectIntFromArray(Direct Array Index=$Particles.PartitionParticles.Partition, Int Selection Array=DataInterface() {propertyValues {IntData [4,5]}})
    MaskSpawnGroup=false
    Mesh Sampling Type=::NewEnumerator5
    Sampled Position Offset=&Lerp_Vector(
     A=&SelectVectorFromArray(
      Array Sampling Mode=::NewEnumerator0
      Direct Array Index=?
      Vector Selection Array=DataInterface() {propertyValues {FloatData [<10 0 -2>, <15 0 -4>, <20 0 0>, <25 0 0>, <30 0 0>]}}
      Random Seed=0
     )
     B=<1 1 1>
     Alpha=&MakeFloatFromInt(Int=$Particles.PartitionParticles.Partition)
    )
    Skeletal Mesh=$System.Skeletal Mesh
    Socket Sampling Mode=::NewEnumerator5
   ) {enabled=false, module=SkeletalMeshLocation001}
   &ScaleMeshSize(Scale Factor=&VectorFromCurve(
     VectorCurve=DataInterface() {propertyValues {
       XCurve {keys [<at 0 1>,<at 0.3426 0.9144>,<at 0.9758 1.3967>]}
       YCurve {keys [<at 0 1>,<at 0.6063 1.3611>,<at 0.9654 1.1803>]}
       ZCurve {keys [<at 0 0.577>,<at 0.1533 1.3432>,<at 0.6073 1.0575>,<at 0.9862 1.2393>]}
      }}
     Scale Curve=&Multiply_VectorByFloat(
      Vector=&RandomRangeVector(Minimum=<0.75 0.75 0.75>, Maximum=<1.25 1.25 1.25>)
      Float=&FloatFromCurve(FloatCurve=DataInterface() {propertyValues {Curve {keys [
           {interpMode=RCIM_Cubic, tangentMode=RCTM_User, time=0, value=0.25}
           {interpMode=RCIM_Cubic, tangentMode=RCTM_User, time=0.2909, value=1}
           {interpMode=RCIM_Cubic, tangentMode=RCTM_User, time=0.8, value=1}
           {arriveTangent=-4.1587, interpMode=RCIM_Cubic, leaveTangent=-4.1587, tangentMode=RCTM_User, time=1, value=-0.0118}
          ]}}})
     )
    )) {enabled=false}
   &ScaleMeshSize(Scale Factor=&VectorFromCurve(
     VectorCurve=DataInterface() {propertyValues {XCurve {keys [<at 0 1>,<at 1 0>]}, YCurve {keys [<at 0 1>,<at 1 0>]}, ZCurve {keys [<at 0 1>,<at 1 0>]}}}
     Scale Curve=&Multiply_VectorByFloat(
      Vector=&RandomRangeVector(Minimum=<0.75 0.75 0.75>, Maximum=<1.5 1.5 1.5>)
      Float=&Lerp_Float(
       A=&FloatFromCurve(FloatCurve=DataInterface() {propertyValues {Curve {keys [
            {interpMode=RCIM_Cubic, tangentMode=RCTM_User, time=0, value=0.25}
            {interpMode=RCIM_Cubic, tangentMode=RCTM_User, time=0.2909, value=1}
            {interpMode=RCIM_Cubic, tangentMode=RCTM_User, time=0.8, value=1}
            {arriveTangent=-4.1587, interpMode=RCIM_Cubic, leaveTangent=-4.1587, tangentMode=RCTM_User, time=1, value=-0.0118}
           ]}}})
       B=&FloatFromCurve(FloatCurve=DataInterface() {propertyValues {Curve {keys [
            {interpMode=RCIM_Cubic, tangentMode=RCTM_User, time=0, value=1}
            {interpMode=RCIM_Cubic, tangentMode=RCTM_User, time=0.2909, value=1}
            {interpMode=RCIM_Cubic, tangentMode=RCTM_User, time=0.8, value=1}
            {arriveTangent=-4.1587, interpMode=RCIM_Cubic, leaveTangent=-4.1587, tangentMode=RCTM_User, time=1, value=-0.0118}
           ]}}})
       Alpha=&MakeFloatFromInt(Int=&SpawnGroup())
      )
     )
    )) {module=ScaleMeshSize001}
   &DynamicMaterialParameters(
    Index 0 Param 1=&FloatFromCurve(FloatCurve=DataInterface() {propertyValues {Curve {keys [<at 0 0>,<at 1 4>]}}})
    Index 0 Param 2=1
    Index 0 Param 3=&FloatFromCurve(
     FloatCurve=DataInterface() {propertyValues {Curve {keys [{interpMode=RCIM_Cubic, tangentMode=RCTM_User, time=0, value=2}, {arriveTangent=-2.5233, tangentMode=RCTM_User, time=1.0059, value=1}]}}}
     Scale Curve=&Waveform(
      "Waveform [1]"=::NewEnumerator9
      "[1] Input Position"=&RandomRangeFloat(Minimum=0, Maximum=1)
      "[1] Frequency"=3
      "[1] Amplitude Min/Max"=<0.75 1>
     )
    )
    Module.Index0Float0Write=true
    Module.Index0Float1Write=true
    Module.Index0Float2Write=true
    Module.Index0Float3Write=true
   )
   &ScaleColor(
    Scale RGB=&Multiply_VectorByFloat(
     Vector=<1 1 1>
     Float=&Lerp_Float(
      A=&FloatFromCurve(FloatCurve=DataInterface() {propertyValues {Curve {keys [<at 0 10>,<at 0.0912 10>,<at 0.2854 2>,<at 0.9021 1.1441>,<at 1 1>]}}})
      B=&FloatFromCurve(FloatCurve=DataInterface() {propertyValues {Curve {keys [<at 0 2>,<at 0.0912 2>,<at 0.2854 2>,<at 0.9021 1.1441>,<at 1 1>]}}})
      Alpha=&MakeFloatFromInt(Int=&SpawnGroup())
     )
    )
    ScaleA=true
    ScaleRGB=true
   )
   /SetVariables(Particles.Position=&AddVectorToPosition(
     Position=$Particles.Position
     Vector=&VectorFromCurve(
      VectorCurve=DataInterface() {propertyValues {
        XCurve {keys [
          {arriveTangent=3.0078, interpMode=RCIM_Cubic, leaveTangent=3.0078, tangentMode=RCTM_User, time=0, value=0}
          {interpMode=RCIM_Cubic, tangentMode=RCTM_User, time=0.5, value=1}
          {arriveTangent=-3.9312, interpMode=RCIM_Cubic, leaveTangent=-3.9312, tangentMode=RCTM_User, time=1, value=0}
         ]}
        YCurve {keys [
          {arriveTangent=3.0078, interpMode=RCIM_Cubic, leaveTangent=3.0078, tangentMode=RCTM_User, time=0, value=0}
          {interpMode=RCIM_Cubic, tangentMode=RCTM_User, time=0.5, value=1}
          {arriveTangent=-3.9312, interpMode=RCIM_Cubic, leaveTangent=-3.9312, tangentMode=RCTM_User, time=1, value=0}
         ]}
        ZCurve {keys [
          {arriveTangent=3.0078, interpMode=RCIM_Cubic, leaveTangent=3.0078, tangentMode=RCTM_User, time=0, value=0}
          {interpMode=RCIM_Cubic, tangentMode=RCTM_User, time=0.5, value=1}
          {arriveTangent=-3.9312, interpMode=RCIM_Cubic, leaveTangent=-3.9312, tangentMode=RCTM_User, time=1, value=0}
         ]}
       }}
      Scale Curve=&Multiply_VectorByFloat(Vector=&RandomRangeVector(Minimum=<-1 -1 -1>, Maximum=<1 1 1>), Float=3)
     )
    )) {isSetParameters=true, module=SetVariables_#4}
  ]
  rendererClasses [&NiagaraSpriteRendererProperties, &NiagaraMeshRendererProperties]
  renderers [
   {rendererClass=&NiagaraSpriteRendererProperties}
   {rendererClass=&NiagaraMeshRendererProperties, rendererIndex=1}
  ]
 }
 IdleCubes001 {
  emitterUpdateScript [
   &EmitterState()
   &SpawnBurst_Instantaneous(
    Spawn Count=5
    Spawn Group=1
    Spawn Probability=&MaskFloat(
     Mask Source=1
     Mask Parameter=&MakeIntFromFloat(Float=$User.Variant, Conversion Mode=::NewEnumerator1)
     Mask=2
     Masked Float=1
     Passthrough Float=0
    )
    Use Spawn Probability=true
    UseLoopCountLimit=true
   )
   &SpawnRate(
    Spawn Probability=&MaskFloat(
     Mask Source=1
     Mask Parameter=&MakeIntFromFloat(Float=$User.Variant, Conversion Mode=::NewEnumerator1)
     Mask=2
     Masked Float=1
     Passthrough Float=0
    )
    SpawnRate=10
    Use Spawn Probability=true
   )
  ]
  particleSpawnScript [
   &PartitionParticles()
   &InitializeParticle(
    Color Maximum=<rgba 1.25 1.1493 0.9825 1>
    Color Minimum=<rgba 0.6 0.6 0.6 1>
    Color Mode=::NewEnumerator2
    Lifetime=&RandomRangeFloat(Minimum=1, Maximum=3)
    Mesh Scale Mode=::NewEnumerator3
    Mesh Uniform Scale=0.7
    UsePositionOffset=false
    Write Material Random=true
   )
   &SubUVAnimation(
    Mesh Renderer=DataInterface() {propertyValues {MeshRenderer=&self:IdleCubes001_0.NiagaraMeshRendererProperties_0}}
    Output Alternate Renderer Binding=true
    Random Start Frame=true
    Renderer Type:ENiagaraMeshOrSprite=::NewEnumerator1
    SubUV Lookup Index Scale=0
    UseEndFrame=false
    UseStartFrame=false
   )
  ]
  particleUpdateScript [
   &ParticleState()
   &SkeletalMeshLocation(
    Bone / Socket Index=3
    MaskSpawnGroup=false
    Mesh Sampling Type=::NewEnumerator5
    Sampled Orientation Coordinate Space=::Local
    Sampled Position Offset=&AddVectorToPosition(
     Position=&SelectVectorFromArray(
      Array Sampling Mode=::NewEnumerator0
      Direct Array Index=?
      Vector Selection Array=DataInterface() {propertyValues {FloatData [<0 0 4>, <0 0 8>, <0 0 12>, <0 0 16>, <0 0 20>]}}
      Random Seed=0
     )
     Vector=&RandomRangeVector(Minimum=<2 0 0>, Maximum=<4 1 1>)
    )
    Skeletal Mesh=$System.Skeletal Mesh
    Socket Sampling Mode=::NewEnumerator5
   ) {enabled=false}
   &SkeletalMeshLocation(
    Bone / Socket Index=3
    MaskSpawnGroup=false
    Mesh Sampling Type=::NewEnumerator5
    Sampled Orientation Coordinate Space=::Local
    Sampled Position Offset=&RandomRangeVector(Minimum=<2 -1 3>, Maximum=<4 1 25>)
    Skeletal Mesh=$System.Skeletal Mesh
    Socket Sampling Mode=::NewEnumerator5
   ) {module=SkeletalMeshLocation002}
   &SkeletalMeshLocation(
    Bone / Socket Index=&SelectIntFromArray(Direct Array Index=$Particles.PartitionParticles.Partition, Int Selection Array=DataInterface() {propertyValues {IntData [4,5]}})
    MaskSpawnGroup=false
    Mesh Sampling Type=::NewEnumerator5
    Sampled Position Offset=&Lerp_Vector(
     A=&SelectVectorFromArray(
      Array Sampling Mode=::NewEnumerator0
      Direct Array Index=?
      Vector Selection Array=DataInterface() {propertyValues {FloatData [<10 0 -2>, <15 0 -4>, <20 0 0>, <25 0 0>, <30 0 0>]}}
      Random Seed=0
     )
     B=<1 1 1>
     Alpha=&MakeFloatFromInt(Int=$Particles.PartitionParticles.Partition)
    )
    Skeletal Mesh=$System.Skeletal Mesh
    Socket Sampling Mode=::NewEnumerator5
   ) {enabled=false, module=SkeletalMeshLocation001}
   &ScaleMeshSize(Scale Factor=&VectorFromCurve(
     VectorCurve=DataInterface() {propertyValues {XCurve {keys [<at 0 1>,<at 1 0>]}, YCurve {keys [<at 0 1>,<at 1 0>]}, ZCurve {keys [<at 0 1>,<at 1 0>]}}}
     Scale Curve=&Multiply_VectorByFloat(
      Vector=&RandomRangeVector(Minimum=<0.75 0.75 0.75>, Maximum=<1.5 1.5 1.5>)
      Float=&Lerp_Float(
       A=&FloatFromCurve(FloatCurve=DataInterface() {propertyValues {Curve {keys [
            {interpMode=RCIM_Cubic, tangentMode=RCTM_User, time=0, value=0.25}
            {interpMode=RCIM_Cubic, tangentMode=RCTM_User, time=0.2909, value=1}
            {interpMode=RCIM_Cubic, tangentMode=RCTM_User, time=0.8, value=1}
            {arriveTangent=-4.1587, interpMode=RCIM_Cubic, leaveTangent=-4.1587, tangentMode=RCTM_User, time=1, value=-0.0118}
           ]}}})
       B=&FloatFromCurve(FloatCurve=DataInterface() {propertyValues {Curve {keys [
            {interpMode=RCIM_Cubic, tangentMode=RCTM_User, time=0, value=1}
            {interpMode=RCIM_Cubic, tangentMode=RCTM_User, time=0.2909, value=1}
            {interpMode=RCIM_Cubic, tangentMode=RCTM_User, time=0.8, value=1}
            {arriveTangent=-4.1587, interpMode=RCIM_Cubic, leaveTangent=-4.1587, tangentMode=RCTM_User, time=1, value=-0.0118}
           ]}}})
       Alpha=&MakeFloatFromInt(Int=&SpawnGroup())
      )
     )
    )) {module=ScaleMeshSize001}
   &DynamicMaterialParameters(
    Index 0 Param 1=&FloatFromCurve(FloatCurve=DataInterface() {propertyValues {Curve {keys [<at 0 0>,<at 1 4>]}}})
    Index 0 Param 2=1
    Index 0 Param 3=&FloatFromCurve(
     FloatCurve=DataInterface() {propertyValues {Curve {keys [{interpMode=RCIM_Cubic, tangentMode=RCTM_User, time=0, value=2}, {arriveTangent=-2.5233, tangentMode=RCTM_User, time=1.0059, value=1}]}}}
     Scale Curve=&Waveform(
      "Waveform [1]"=::NewEnumerator9
      "[1] Input Position"=&RandomRangeFloat(Minimum=0, Maximum=1)
      "[1] Frequency"=3
      "[1] Amplitude Min/Max"=<0.75 1>
     )
    )
    Module.Index0Float0Write=true
    Module.Index0Float1Write=true
    Module.Index0Float2Write=true
    Module.Index0Float3Write=true
   )
   &ScaleColor(
    Scale RGB=&Multiply_VectorByFloat(
     Vector=<1 1 1>
     Float=&Lerp_Float(
      A=&FloatFromCurve(FloatCurve=DataInterface() {propertyValues {Curve {keys [
           <at 0 1>
           {interpMode=RCIM_Cubic, tangentMode=RCTM_User, time=0.2854, value=1}
           {arriveTangent=-1.8481, tangentMode=RCTM_User, time=1, value=0.25}
          ]}}})
      B=&FloatFromCurve(FloatCurve=DataInterface() {propertyValues {Curve {keys [<at 0 2>,<at 0.0912 2>,<at 0.2854 2>,<at 0.9021 1.1441>,<at 1 1>]}}})
      Alpha=&MakeFloatFromInt(Int=&SpawnGroup())
     )
    )
    ScaleA=true
    ScaleRGB=true
   )
   /SetVariables(Particles.Position=&AddVectorToPosition(
     Position=$Particles.Position
     Vector=&Multiply_VectorByFloat(
      Vector=<10 0 10>
      Float=&FloatFromCurve(
       FloatCurve=DataInterface() {propertyValues {Curve {keys [{interpMode=RCIM_Cubic, tangentMode=RCTM_User, time=0, value=0}, <at 1 1>]}}}
       Scale Curve=&RandomRangeFloat(Minimum=0.8, Maximum=1.2)
      )
     )
    )) {isSetParameters=true, module=SetVariables_#4}
   &UpdateMeshOrientation(
    Rotation Rate=&FloatFromCurve(
     FloatCurve=DataInterface() {propertyValues {Curve {keys [{interpMode=RCIM_Cubic, tangentMode=RCTM_User, time=0.25, value=0}, <at 1 1>]}}}
     Scale Curve=10
    )
    Rotation Vector=&RandomRangeVector(Minimum=<-1 -1 -1>, Maximum=<1 1 1>)
   )
   &BlendSpace("Position Blend (World <-> Local)":NiagaraFloat=&FloatFromCurve(
     FloatCurve=DataInterface() {propertyValues {Curve {keys [{interpMode=RCIM_Cubic, tangentMode=RCTM_User, time=0, value=1}, {arriveTangent=-0.9914, tangentMode=RCTM_User, time=1, value=0}]}}}
     Translation Contribution:NiagaraFloat=1
     Rotation Contribution:NiagaraFloat=1
    ))
  ]
  rendererClasses [&NiagaraSpriteRendererProperties, &NiagaraMeshRendererProperties]
  renderers [
   {rendererClass=&NiagaraSpriteRendererProperties}
   {rendererClass=&NiagaraMeshRendererProperties, rendererIndex=1}
  ]
 }
 Smoke {
  emitterUpdateScript [
   &EmitterState(Inactive Response=::NewEnumerator1, Life Cycle Mode=::NewEnumerator1, Loop Duration=1, UseLoopDelay=false)
   &SpawnParticlesFromOtherEmitter(
    Attribute Reader=DataInterface() {propertyValues {EmitterBinding {bindingMode=Other, emitterName=Arks}}}
    Spawn Rate=5
    Use Maximum Age=false
    Use Minimum Age=false
   )
  ]
  particleSpawnScript [
   &PartitionParticles(Number Of Partitions=$System.SpawnCount, Partition Mode:ENiagara_PartitionMode=::NewEnumerator5)
   &InitializeParticle(
    Color=&MaskLinearColor(
     Mask Source=1
     Mask Parameter=&MakeIntFromFloat(Float=$User.Variant, Conversion Mode=::NewEnumerator1)
     Mask=1
     Masked Linear Color=<rgba 0.2806 1 0.9701 0.1>
     Passthrough Linear Color=<rgba 1 0.5925 0.1614 0.1>
    )
    Lifetime Max=0.8
    Lifetime Min=0.3
    Lifetime Mode=::NewEnumerator1
    Sprite Size Mode=::NewEnumerator1
    Uniform Sprite Size Max=6
    Uniform Sprite Size Min=3
    UsePositionOffset=false
    Write Material Random=true
   )
   &SampleParticlesFromOtherEmitter(Reject Particle on Sampled Particle's First Frame=true, Transform Data Option=1)
  ]
  particleUpdateScript [
   &ParticleState()
   &ScaleSpriteSize(
    Scale Sprite Size Mode=::NewEnumerator0
    Uniform Scale Factor=&MakeFloatFromVector(Vector=$Engine.Owner.Scale, Channel=::NewEnumerator2)
   ) {enabled=false}
   &DynamicMaterialParameters(
    Index 0 Param 1=&FloatFromCurve(FloatCurve=DataInterface() {propertyValues {Curve {keys [{tangentMode=RCTM_User, time=0, value=0}, {tangentMode=RCTM_User, time=1, value=0.5}]}}})
    Index 0 Param 2=0.5
    Index 0 Param 4=1
    Module.Index0Float0Write=true
    Module.Index0Float1Write=true
    Module.Index0Float2Write=true
    Module.Index0Float3Write=true
    Module.Index1Float0Write=true
    Module.Index1Float1Write=true
    Module.Index1Float2Write=true
    Module.Index1Float3Write=true
    Write Parameter Index 1=::NewEnumerator1
   ) {module=DynamicMaterialParameters001}
   &ScaleSpriteSize(
    Uniform Curve Scale=2
    Uniform Curve Sprite Scale=DataInterface() {propertyValues {Curve {keys [
        {arriveTangent=4.1785, interpMode=RCIM_Cubic, leaveTangent=4.1785, tangentMode=RCTM_User, time=0, value=0}
        {arriveTangent=1.1179, interpMode=RCIM_Cubic, leaveTangent=1.1179, tangentMode=RCTM_User, time=0.2042, value=0.6228}
        {arriveTangent=0.0224, interpMode=RCIM_Cubic, leaveTangent=0.0224, tangentMode=RCTM_User, time=1, value=1}
       ]}}}
   ) {module=ScaleSpriteSize001}
   &Drag(Drag=500, Use Linear Drag=true, Use Rotational Drag=true)
   &CurlNoiseForce(Noise Strength=100, Pan Noise=false)
   &ScaleColor(
    Scale Alpha=&FloatFromCurve(FloatCurve=DataInterface() {propertyValues {Curve {keys [
         {arriveTangent=9.0725, interpMode=RCIM_Cubic, leaveTangent=9.0725, tangentMode=RCTM_User, time=0, value=0}
         {arriveTangent=-2.5249, interpMode=RCIM_Cubic, leaveTangent=-2.5249, tangentMode=RCTM_User, time=0.1408, value=0.9847}
         {arriveTangent=-0.0405, interpMode=RCIM_Cubic, leaveTangent=-0.0405, tangentMode=RCTM_User, time=1, value=0}
        ]}}})
    ScaleA=true
    ScaleRGB=true
   )
   &SolveForcesAndVelocity(Clamp Velocity=false, Limit Acceleration=false)
  ]
  rendererClasses [&NiagaraSpriteRendererProperties]
  renderers [{rendererClass=&NiagaraSpriteRendererProperties}]
 }
 Sparks {
  emitterUpdateScript [
   &EmitterState(Inactive Response=::NewEnumerator1, Life Cycle Mode=::NewEnumerator1, Loop Duration=1, UseLoopDelay=false)
   &self:ReduceSpawnWhileRunning(Speed=300) {module=ReduceSpawnWhileRunning}
   &SpawnParticlesFromOtherEmitter(
    Attribute Reader=DataInterface() {propertyValues {EmitterBinding {bindingMode=Other, emitterName=Arks}}}
    Minimum Age Spawn Threshold=0
    Spawn Rate=&FloatFromCurve(FloatCurve=DataInterface() {propertyValues {Curve {keys [<at 0 1>,<at 1 0.5>]}}}, CurveIndex=$Emitter.SpawnMultiply, Scale Curve=12)
    Use Maximum Age=false
    Use Minimum Age=true
   )
  ]
  particleSpawnScript [
   &PartitionParticles(Number Of Partitions=$System.SpawnCount, Partition Mode:ENiagara_PartitionMode=::NewEnumerator5)
   &InitializeParticle(
    Color=&MaskLinearColor(
     Mask Source=1
     Mask Parameter=&MakeIntFromFloat(Float=$User.Variant)
     Mask=1
     Masked Linear Color=&SelectLinearColorFromArray(
      Array Sampling Mode=::NewEnumerator0
      Direct Array Index=?
      Color Selection Array=DataInterface() {propertyValues {ColorData [<rgba 0 10 10 0.302>, <rgba 0 7.225 10 0.302>, <rgba 0 2.4407 10 1>]}}
      Random Seed=0
     )
     Passthrough Linear Color=&SelectLinearColorFromArray(
      Array Sampling Mode=::NewEnumerator0
      Direct Array Index=?
      Color Selection Array=DataInterface() {propertyValues {ColorData [<rgba 10 5.6687 0.1522 1>, <rgba 10 3.3 0.287 1>, <rgba 10 2.5941 0.077 1>]}}
      Random Seed=0
     )
    )
    Lifetime Max=0.35
    Lifetime Min=0.2
    Lifetime Mode=::NewEnumerator1
    Sprite Rotation Angle=&SelectFloatFromArray(Direct Array Index=$Particles.PartitionParticles.Partition, Float Selection Array=DataInterface() {propertyValues {FloatData [90,-90,0,0]}})
    Sprite Rotation Mode=::NewEnumerator0
    Sprite Size Max=<1.5 5>
    Sprite Size Min=<1 3>
    Sprite Size Mode=::NewEnumerator2
    UsePositionOffset=false
    Write Material Random=true
   )
   &AddVelocity(
    Cone Axis=<-1 0 0>
    Module.UseConeFalloff=false
    Module.UseVelDistribution=false
    Override Seed=false
    Velocity Mode=::NewEnumerator2
    Velocity Speed=100
   ) {module=AddVelocity001}
   &SampleParticlesFromOtherEmitter(Reject Particle on Sampled Particle's First Frame=true, Transform Data Option=1)
  ]
  particleUpdateScript [
   &ParticleState()
   &ScaleSpriteSize(Uniform Curve Sprite Scale=DataInterface() {propertyValues {Curve {keys [
        {arriveTangent=-1.7874, interpMode=RCIM_Cubic, leaveTangent=-1.7874, tangentMode=RCTM_User, time=0, value=1}
        {tangentMode=RCTM_User, time=1, value=0}
       ]}}})
   &GravityForce(Gravity=&Multiply_VectorByFloat(
     Vector=<1 1 -350>
     Float=&FloatFromCurve(FloatCurve=DataInterface() {propertyValues {Curve {keys [{interpMode=RCIM_Cubic, tangentMode=RCTM_User, time=0, value=0}, {interpMode=RCIM_Cubic, tangentMode=RCTM_User, time=1, value=1}]}}})
    ))
   &ScaleColor(
    Scale Alpha=&FloatFromCurve(FloatCurve=DataInterface() {propertyValues {Curve {keys [
         {tangentMode=RCTM_User, time=0, value=1}
         {arriveTangent=-0.0405, interpMode=RCIM_Cubic, leaveTangent=-0.0405, tangentMode=RCTM_User, time=0.7009, value=1}
         {arriveTangent=-8.0089, interpMode=RCIM_Cubic, leaveTangent=-8.0089, tangentMode=RCTM_User, time=1, value=0}
        ]}}})
    ScaleA=true
    ScaleRGB=true
   )
   &SolveForcesAndVelocity(Clamp Velocity=false, Limit Acceleration=false)
  ]
  rendererClasses [&NiagaraSpriteRendererProperties]
  renderers [{rendererClass=&NiagaraSpriteRendererProperties}]
 }
}
summary {
 systemName=NS_Pickaxe_SpacePlunge_Idle
 emitters [
  {emitterName=Arks, rendererClasses [&NiagaraSpriteRendererProperties, &NiagaraLightRendererProperties]}
  {emitterName=Sparks, rendererClasses [&NiagaraSpriteRendererProperties]}
  {emitterName=Smoke, rendererClasses [&NiagaraSpriteRendererProperties]}
  {
   bEnabled=false
   emitterName=IdleCubes
   rendererClasses [&NiagaraSpriteRendererProperties, &NiagaraMeshRendererProperties]
  }
  {emitterName=IdleCubes001, rendererClasses [&NiagaraSpriteRendererProperties, &NiagaraMeshRendererProperties]}
 ]
 userVariables [
  User.bIsFrontEnd {type=NiagaraBool}
  User.bIsFrontEndPreview {type=NiagaraBool}
  User.bIsSwinging {type=NiagaraBool}
  User.Variant {type=NiagaraFloat, defaultValue=0, description=""}
 ]
}
