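# Control Rig for the SpacePlunge pickaxe that spins the dyn_Saw bone at speeds driven by the weapon's state.
# Idle and swinging flags select a vector packing speed, duration and blend, while jumping and sprinting branches exist but are not connected.
# A looping timer latches duration and target speed near each cycle's end and refreshes blend inputs during its first half.
# An accumulated lerp smooths the speed feeding a 359-degree time loop that becomes the bone's Y rotation.
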ControlRigBlueprint /BRCosmetics/Weapons/FORT_Melee/Pickaxe_SpacePlunge/Meshes/SpacePlunge_Axe_CtrlRig
graphs [
 RigVMFunctionLibrary {links [], nodes []}
 RigVMModel {nodes [
   RigUnit_BeginExecution = &RigUnit_BeginExecution()
   RigVMFunction_Sequence = &RigVMFunction_Sequence(ExecuteContext=@RigUnit_BeginExecution.ExecutePin)
   n1 = &RigVMFunction_ControlFlowBranch(ExecuteContext=@RigVMFunction_Sequence.A, Condition=RigVMVariableNode() {variable=isIdle}.Value)
   RigVMFunction_ControlFlowBranch = &RigVMFunction_ControlFlowBranch(ExecuteContext=@n1.Completed, Condition=RigVMVariableNode() {variable=isSwinging}.Value)
   RigVMVariableNode(Value=(X=450,Y=2.05,Z=1), ExecuteContext=@RigVMFunction_ControlFlowBranch.True) {variable=Speed and Duration}
   RigVMVariableNode(Value=(X=10,Y=0.1,Z=0.75), ExecuteContext=@n1.True) {variable=Speed and Duration}
   n2 = RigVMVariableNode() {variable=Initial Duration}
   n3 = /Script/RigVM.RigVMFunction_TimeLoop(Normalize=false, Speed=1, Duration=@n2.Value) {is=struct RigVMUnitNode}
   n4 = &RigVMFunction_ControlFlowBranch(
    Condition=/Script/RigVM.RigVMFunction_MathDoubleGreaterEqual(A=@n3.Relative, B=&RigVMFunction_MathDoubleSub(B=0.1, A=@n2.Value).Result) {is=struct RigVMUnitNode, template="GreaterEqual::Execute(in A,in B,out Result)"}.Result
    ExecuteContext=@RigVMFunction_Sequence.B
   )
   Multiply = &RigVMFunction_MathDoubleMul(B=0.5, A=@n2.Value)
   n5 = &RigVMFunction_ControlFlowBranch(
    Condition=&RigVMFunction_MathDoubleGreaterEqual(B=@Multiply.Result, A=@n3.Relative) {template="GreaterEqual::Execute(in A,in B,out Result)"}.Result
    ExecuteContext=@n4.Completed
   )
   n6 = RigVMVariableNode() {variable=Speed and Duration}
   RigVMVariableNode(ExecuteContext=@n5.True, Value=@n6.Value.X) {variable=Next Speed}
   > RigVMVariableNode(Value=@n6.Value.Z) {variable=blend state}
   AccumulateLerp = /Script/RigVM.RigVMFunction_AccumulateFloatLerp(
    bIntegrateDeltaTime=true
    TargetValue=RigVMVariableNode() {variable=Current Speed}.Value
    Blend=RigVMVariableNode() {variable=blend state}.Value
    InitialValue=RigVMVariableNode() {variable=Next Speed}.Value
   ) {is=struct RigVMUnitNode, template="AccumulateLerp::Execute(in TargetValue,in InitialValue,in Blend,in bIntegrateDeltaTime,out Result)"}
   RigVMFunction_TimeLoop = &RigVMFunction_TimeLoop(Duration=359, Normalize=false, Speed=@AccumulateLerp.Result)
   RigVMFunction_MathQuaternionFromEuler = &RigVMFunction_MathQuaternionFromEuler(Euler=(X=0,Y=0,Z=0), Euler.Y=&RigVMFunction_MathDoubleMul(B=1, A=@RigVMFunction_TimeLoop.Relative).Result)
   &RigUnit_ModifyTransforms(
    ItemToModify=((Item=(Type=Bone,Name="dyn_Saw"),Transform=(Rotation=(X=0,Y=0,Z=0,W=1),Translation=(X=0,Y=0,Z=0),Scale3D=(X=1,Y=1,Z=1))))
    Weight=1
    WeightMinimum=0
    ItemToModify.0.Transform.Rotation=@RigVMFunction_MathQuaternionFromEuler.Result
    ExecutePin=@n5.Completed
   )
   RigVMVariableNode(Value=RigVMVariableNode() {variable=Initial Duration}.Value, ExecuteContext=@n4.False) {variable=Initial Duration}
   RigVMVariableNode(Value=@n6.Value.Y, ExecuteContext=@n4.True) {variable=Initial Duration}
   > RigVMVariableNode(Value=@n6.Value.Y) {variable=Current Duration}
   > RigVMVariableNode(Value=@n6.Value.X) {variable=Current Speed}
   n7 = &RigVMFunction_ControlFlowBranch(ExecuteContext=(), Condition=RigVMVariableNode() {variable=isJumping}.Value)
   n8 = &RigVMFunction_ControlFlowBranch(ExecuteContext=@n7.Completed, Condition=RigVMVariableNode() {variable=isSprinting}.Value)
   RigVMVariableNode(Value=(X=300,Y=2,Z=1), ExecuteContext=@n8.True) {variable=Speed and Duration}
   RigVMVariableNode(Value=(X=300,Y=2,Z=3), ExecuteContext=@n7.True) {variable=Speed and Duration}
   RigVMVariableNode(ExecuteContext=(), Value=(X=100,Y=0.1,Z=1)) {variable=Speed and Duration}
   RigVMVariableNode(ExecuteContext=(), Value=(X=600,Y=1,Z=2)) {variable=Speed and Duration}
   RigVMVariableNode(ExecuteContext=(), Value=(X=50,Y=0.1,Z=0.1)) {variable=Speed and Duration}
  ]}
]
hierarchy [
 b_AR_Root {loc=<0 0 0>, rot=<0 0 0 1>}
 Melee_TwoHanded_Handle {parent=b_AR_Root, loc=<0 0 0>, rot=<0 0 0.7071 -0.7071>}
 Grip {parent=Melee_TwoHanded_Handle, loc=<0 0 -0.0156>, rot=<0 0 -0.7071 -0.7071>}
 base_Handle_FlexPoint {parent=Grip, loc=<-0.9739 0 20.2545>, rot=<0 0.5251 0 -0.851>}
 dyn_Handle_FlexPoint {parent=base_Handle_FlexPoint, loc=<0 0 0>, rot=<0 0 0 1>}
 base_SawBase {parent=dyn_Handle_FlexPoint, loc=<16.6143 0 0.51>, rot=<0 0.3498 0 -0.9368>}
 dyn_SawBase {parent=base_SawBase, loc=<0 0 0>, rot=<0 0 0 1>}
 base_Saw {parent=dyn_SawBase, loc=<12.2488 0 0>, rot=<0 0 0 1>}
 dyn_Saw {parent=base_Saw, loc=<0 0 0>, rot=<0 0 0 1>}
 base_Cable_Short_01 {parent=Grip, loc=<-0.0072 -0.0028 -12.2543>, rot=<-0.0005 -0.7099 -0.0005 -0.7043>}
 dyn_Cable_Short_01 {parent=base_Cable_Short_01, loc=<0 0 0>, rot=<0 0 0 1>}
 dyn_Cable_Short_02 {parent=dyn_Cable_Short_01, loc=<3.9464 0 0>, rot=<0 0 0 1>}
 base_Cable_Long_01 {parent=Grip, loc=<1.4502 0.0081 -12.2827>, rot=<0 -0.7069 0 -0.7073>}
 dyn_Cable_Long_01 {parent=base_Cable_Long_01, loc=<0 0 0>, rot=<0 0 0 1>}
 dyn_Cable_Long_02 {parent=dyn_Cable_Long_01, loc=<3.8927 0 0>, rot=<0 0 0.0007 -1>}
 dyn_Cable_Long_03 {parent=dyn_Cable_Long_02, loc=<5.7562 0 0>, rot=<0 0 0 1>}
]
variables [
 Current Duration {type=double, default=5}
 Current Speed {type=double, default=0}
 Initial Duration {type=double, default=1}
 Next Speed {type=double, default=0}
 Speed and Duration {type=FVector, default=(X=10,Y=0.1,Z=0.1)}
 blend state {type=double, default=0}
 isIdle {type=bool, default=False}
 isJumping {type=bool, default=False}
 isSprinting {type=bool, default=False}
 isSwinging {type=bool, default=False}
]
