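# Physics asset for the SpacePlunge pickaxe that gives capsule bodies to its grip and two hanging cables, one long and one short.
# Each cable is a chain from a small base body constrained to the grip, with three dynamic segments on the long cable and two on the short.
# The base-to-grip joints are free in swing and twist, while the joints down each chain have restricted swing and zero twist range.
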PhysicsAsset /BRCosmetics/Weapons/FORT_Melee/Pickaxe_SpacePlunge/Meshes/SpacePlunge_Axe_Physics
bodies {
 Grip [{
   shapeType=Capsule
   length=4.0492
   radius=8.3264
   shapeName=Grip_capsule
   center=<-2.7522 -0.0025 -4.0329>
   rotation=<rot -0.0001 -90 69.0656>
  }]
 base_Cable_Long_01 [{shapeType=Capsule, length=0.505, radius=0.505, shapeName=base_Cable_Long_01_capsule, rotation=<rot 0 -90 0>}]
 base_Cable_Short_01 [{shapeType=Capsule, length=0.505, radius=0.505, shapeName=base_Cable_Short_01_capsule, rotation=<rot 0 -90 0>}]
 dyn_Cable_Long_01 [{
   shapeType=Capsule
   length=3.5733
   radius=0.5703
   shapeName=dyn_Cable_Long_01_capsule
   center=<2.4715 -0.0081 -0.0511>
   rotation=<rot 0 90 -89.9974>
  }]
 dyn_Cable_Long_02 [{
   shapeType=Capsule
   length=5.6991
   radius=0.5798
   shapeName=dyn_Cable_Long_02_capsule
   center=<3.0308 -0.004 -0.0519>
   rotation=<rot 0 -89.9219 -89.9929>
  }]
 dyn_Cable_Long_03 [{
   shapeType=Capsule
   length=1.9757
   radius=0.8068
   shapeName=dyn_Cable_Long_03_capsule
   center=<1.8073 0.0042 -0.0526>
   rotation=<rot 0 -90 90>
  }]
 dyn_Cable_Short_01 [{
   shapeType=Capsule
   length=3.5732
   radius=0.6147
   shapeName=dyn_Cable_Short_01_capsule
   center=<2.4999 -0.0009 0.0045>
   rotation=<rot 0 89.9157 -89.518>
  }]
 dyn_Cable_Short_02 [{
   shapeType=Capsule
   length=3.091
   radius=0.7164
   shapeName=dyn_Cable_Short_02_capsule
   center=<2.4954 -0.0067 0.0484>
   rotation=<rot 0 -90.0843 -89.906>
  }]
}
constraints [
 {
  bone1Name=base_Cable_Long_01
  bone2Name=Grip
  swing1LimitDegrees=45
  swing1Motion=Free
  swing2LimitDegrees=45
  swing2Motion=Free
  twistLimitDegrees=45
  twistMotion=Free
 }
 {bone1Name=dyn_Cable_Long_01, bone2Name=base_Cable_Long_01, swing1LimitDegrees=30, swing2LimitDegrees=30, twistLimitDegrees=0}
 {bone1Name=dyn_Cable_Long_02, bone2Name=dyn_Cable_Long_01, swing1LimitDegrees=30, swing2LimitDegrees=30, twistLimitDegrees=0}
 {bone1Name=dyn_Cable_Long_03, bone2Name=dyn_Cable_Long_02, swing1LimitDegrees=30, swing2LimitDegrees=30, twistLimitDegrees=0}
 {
  bone1Name=base_Cable_Short_01
  bone2Name=Grip
  swing1LimitDegrees=45
  swing1Motion=Free
  swing2LimitDegrees=45
  swing2Motion=Free
  twistLimitDegrees=45
  twistMotion=Free
 }
 {bone1Name=dyn_Cable_Short_01, bone2Name=base_Cable_Short_01, swing1LimitDegrees=30, swing2LimitDegrees=30, twistLimitDegrees=0}
 {bone1Name=dyn_Cable_Short_02, bone2Name=dyn_Cable_Short_01, swing1LimitDegrees=30, swing2LimitDegrees=30, twistLimitDegrees=0}
]
