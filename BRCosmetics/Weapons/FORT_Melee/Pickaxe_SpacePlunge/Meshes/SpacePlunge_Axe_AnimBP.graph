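# Animation Blueprint for the SpacePlunge pickaxe, adding secondary motion to its handle, saw base and short and long cables.
# A Control Rig pose feeds spring Anim Dynamics on handle and saw, cached for selectable cable branches: front-end/in-game-tuned RigidBody, Anim Dynamics chains, or additive bone poses.
# A final bool blend can add cable-base offsets.
# Each update sets state-based gravity, front-end preview dynamics weights and an idle flag from emote/melee state.
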AnimBlueprint /BRCosmetics/Weapons/FORT_Melee/Pickaxe_SpacePlunge/Meshes/SpacePlunge_Axe_AnimBP
target_skeleton=&/SpacePlunge_Axe_Skeleton
graphs [
 AnimGraph {nodes [
   n1 = "Transform (Modify) Bone - Bone: base_Cable_Short_01"!(ComponentPose=Local To Component(LocalPose=Use cached pose 'DynamicsCache'()), Alpha=1, Translation="0,0,0", Rotation="30,-5,-40") -> (Pose) {
    rotationMode=BMM_Additive
    rotationSpace=BCS_BoneSpace
    translationMode=BMM_Additive
    translationSpace=BCS_BoneSpace
    boneToModify {boneName=base_Cable_Short_01}
    rotation=<rot 30 -5 -40>
   }
   n2 = "Transform (Modify) Bone - Bone: base_Cable_Long_01"!(ComponentPose=@n1, Alpha=1, Translation="0,0,0", Rotation="20,-5,-50") -> (Pose) {
    rotationMode=BMM_Additive
    rotationSpace=BCS_BoneSpace
    translationMode=BMM_Additive
    translationSpace=BCS_BoneSpace
    boneToModify {boneName=base_Cable_Long_01}
    rotation=<rot 20 -5 -50>
   }
   Output Pose(Result=Blend Poses by bool(BlendPose_0=Use cached pose 'DynamicsCache'(), BlendPose_1=Component To Local(ComponentPose=@n2))) {layerGroup=None}
   AnimGraphNode_SubInput_0 = Input Pose()
   n3 = Control Rig!(Source=@AnimGraphNode_SubInput_0, Alpha, isSwinging:bool=false, isIdle:bool=false) -> (Pose) {controlRigAssetReference {blueprintRigClass=&/SpacePlunge_Axe_CtrlRig.SpacePlunge_Axe_CtrlRig_C}}
   Save cached pose 'CRCache_Input'(Pose=@n3) {cacheName=CRCache_Input}
   n4 = Anim Dynamics - Bone: dyn_Cable_Short_01!(ComponentPose=Local To Component(LocalPose=Use cached pose 'GripDynamics_Cache'()), Alpha=1, GravityOverride:Vector="0,0,-980") -> (Pose) {
    angularDampingOverride=1
    angularSpringConstant=5000
    bAngularSpring=true
    bOverrideAngularDamping=true
    bUseGravityOverride=true
    gravityScale=0.35
    simulationSpace=RootRelative
    boundBone {boneName=dyn_Cable_Short_01}
    componentAppliedLinearAccClamp=<5000 5000 5000>
    componentLinearAccScale=<1 1 1>
    componentLinearVelScale=<1 1 1>
    externalForce {z=-600}
    gravityOverride {z=-980}
    physicsBodyDefinitions [{
      boundBone {boneName=dyn_Cable_Short_01}
      boxExtents=<30 30 30>
      constraintSetup {angularLimitsMax=<0 35 25>, angularLimitsMin=<0 -65 -25>, angularTarget=<1 0 -1>}
      localJointOffset=<3 0 0>
     }]
    relativeSpaceBone {boneName=Melee_TwoHanded_Handle}
    simSpaceSettings {maxAngularAcceleration=25, maxAngularVelocity=5, simSpaceAngularAlpha=1}
   }
   n5 = Anim Dynamics - Bone: dyn_Cable_Short_02!(ComponentPose=@n4, Alpha=1, GravityOverride:Vector="0,0,-980") -> (Pose) {
    angularDampingOverride=1
    angularSpringConstant=5000
    bAngularSpring=true
    bOverrideAngularDamping=true
    bUseGravityOverride=true
    gravityScale=0.35
    simulationSpace=RootRelative
    boundBone {boneName=dyn_Cable_Short_02}
    componentAppliedLinearAccClamp=<5000 5000 5000>
    componentLinearAccScale=<1 1 1>
    componentLinearVelScale=<1 1 1>
    externalForce {z=-600}
    gravityOverride {z=-980}
    physicsBodyDefinitions [{
      boundBone {boneName=dyn_Cable_Short_02}
      boxExtents=<30 30 30>
      constraintSetup {angularLimitsMax=<0 35 25>, angularLimitsMin=<0 -45 -25>, angularTarget=<1 0 -1>}
      localJointOffset=<3 0 0>
     }]
    relativeSpaceBone {boneName=Melee_TwoHanded_Handle}
    simSpaceSettings {maxAngularAcceleration=25, maxAngularVelocity=5, simSpaceAngularAlpha=1}
   }
   n6 = Anim Dynamics - Bone: dyn_Cable_Long_01!(ComponentPose=@n5, Alpha=1, GravityOverride:Vector="0,0,-980") -> (Pose) {
    angularDampingOverride=1
    angularSpringConstant=5000
    bAngularSpring=true
    bOverrideAngularDamping=true
    bUseGravityOverride=true
    gravityScale=0.35
    simulationSpace=RootRelative
    boundBone {boneName=dyn_Cable_Long_01}
    componentAppliedLinearAccClamp=<5000 5000 5000>
    componentLinearAccScale=<1 1 1>
    componentLinearVelScale=<1 1 1>
    externalForce {z=-600}
    gravityOverride {z=-980}
    physicsBodyDefinitions [{
      boundBone {boneName=dyn_Cable_Long_01}
      boxExtents=<30 30 30>
      constraintSetup {angularLimitsMax=<0 25 25>, angularLimitsMin=<0 -75 -25>, angularTarget=<1 0 -1>}
      localJointOffset=<3 0 0>
     }]
    relativeSpaceBone {boneName=Melee_TwoHanded_Handle}
    simSpaceSettings {maxAngularAcceleration=25, maxAngularVelocity=5, simSpaceAngularAlpha=1}
   }
   n7 = Anim Dynamics - Bone: dyn_Cable_Long_02!(ComponentPose=@n6, Alpha=1, GravityOverride:Vector="0,0,-980") -> (Pose) {
    angularDampingOverride=1
    angularSpringConstant=5000
    bAngularSpring=true
    bOverrideAngularDamping=true
    bUseGravityOverride=true
    gravityScale=0.35
    simulationSpace=RootRelative
    boundBone {boneName=dyn_Cable_Long_02}
    componentAppliedLinearAccClamp=<5000 5000 5000>
    componentLinearAccScale=<1 1 1>
    componentLinearVelScale=<1 1 1>
    externalForce {z=-600}
    gravityOverride {z=-980}
    physicsBodyDefinitions [{
      boundBone {boneName=dyn_Cable_Long_02}
      boxExtents=<30 30 30>
      constraintSetup {angularLimitsMax=<0 25 25>, angularLimitsMin=<0 -45 -25>, angularTarget=<1 0 -1>}
      localJointOffset=<3 0 0>
     }]
    relativeSpaceBone {boneName=Melee_TwoHanded_Handle}
    simSpaceSettings {maxAngularAcceleration=25, maxAngularVelocity=5, simSpaceAngularAlpha=1}
   }
   n8 = Anim Dynamics - Bone: dyn_Cable_Long_03!(ComponentPose=@n7, Alpha=1, GravityOverride:Vector="0,0,-980") -> (Pose) {
    angularDampingOverride=1
    angularSpringConstant=2500
    bAngularSpring=true
    bOverrideAngularDamping=true
    bUseGravityOverride=true
    gravityScale=0.35
    simulationSpace=RootRelative
    boundBone {boneName=dyn_Cable_Long_03}
    componentAppliedLinearAccClamp=<5000 5000 5000>
    componentLinearAccScale=<1 1 1>
    componentLinearVelScale=<1 1 1>
    externalForce {z=-600}
    gravityOverride {z=-980}
    physicsBodyDefinitions [{
      boundBone {boneName=dyn_Cable_Long_03}
      boxExtents=<30 30 30>
      constraintSetup {angularLimitsMax=<0 25 25>, angularLimitsMin=<0 -45 -25>, angularTarget=<1 0 -1>}
      localJointOffset=<3 0 0>
     }]
    relativeSpaceBone {boneName=Melee_TwoHanded_Handle}
    simSpaceSettings {maxAngularAcceleration=25, maxAngularVelocity=5, simSpaceAngularAlpha=1}
   }
   n9 = Component To Local(ComponentPose=@n8)
   n10 = "Transform (Modify) Bone - Bone: base_Cable_Short_01"(ComponentPose=Local To Component(LocalPose=Use cached pose 'GripDynamics_Cache'()), Alpha=1, Translation="0,0,0", Rotation="-50,0,0") {
    rotationMode=BMM_Additive
    rotationSpace=BCS_BoneSpace
    translationMode=BMM_Additive
    translationSpace=BCS_BoneSpace
    boneToModify {boneName=base_Cable_Short_01}
    rotation {pitch=-50}
   }
   n11 = "Transform (Modify) Bone - Bone: dyn_Cable_Short_02"!(ComponentPose=@n10, Alpha=1, Translation="0,0,0", Rotation="-50,0,0") -> (Pose) {
    rotationMode=BMM_Additive
    rotationSpace=BCS_BoneSpace
    translationMode=BMM_Additive
    translationSpace=BCS_BoneSpace
    boneToModify {boneName=dyn_Cable_Short_02}
    rotation {pitch=-50}
   }
   n12 = "Transform (Modify) Bone - Bone: base_Cable_Long_01"(ComponentPose=@n11, Alpha=1, Translation="0,0,0", Rotation="-30,0,0") {
    rotationMode=BMM_Additive
    rotationSpace=BCS_BoneSpace
    translationMode=BMM_Additive
    translationSpace=BCS_BoneSpace
    boneToModify {boneName=base_Cable_Long_01}
    rotation {pitch=-30}
   }
   n13 = "Transform (Modify) Bone - Bone: dyn_Cable_Long_02"!(ComponentPose=@n12, Alpha=1, Translation="0,0,0", Rotation="-50,0,0") -> (Pose) {
    rotationMode=BMM_Additive
    rotationSpace=BCS_BoneSpace
    translationMode=BMM_Additive
    translationSpace=BCS_BoneSpace
    boneToModify {boneName=dyn_Cable_Long_02}
    rotation {pitch=-50}
   }
   n14 = "Transform (Modify) Bone - Bone: dyn_Cable_Long_03"!(ComponentPose=@n13, Alpha=1, Translation="0,0,0", Rotation="-40,0,0") -> (Pose) {
    rotationMode=BMM_Additive
    rotationSpace=BCS_BoneSpace
    translationMode=BMM_Additive
    translationSpace=BCS_BoneSpace
    boneToModify {boneName=dyn_Cable_Long_03}
    rotation {pitch=-40}
   }
   n15 = "Transform (Modify) Bone - Bone: base_Cable_Short_01"(ComponentPose=@n14, Alpha=1, Translation="0,0,0", Rotation="30,30,0") {
    rotationMode=BMM_Additive
    rotationSpace=BCS_BoneSpace
    translationMode=BMM_Additive
    translationSpace=BCS_BoneSpace
    boneToModify {boneName=base_Cable_Short_01}
    rotation {pitch=30, yaw=30}
   }
   n16 = "Transform (Modify) Bone - Bone: dyn_Cable_Short_01"!(ComponentPose=@n15, Alpha=1, Translation="0,0,0", Rotation="-10,0,0") -> (Pose) {
    rotationMode=BMM_Additive
    rotationSpace=BCS_BoneSpace
    translationMode=BMM_Additive
    translationSpace=BCS_BoneSpace
    boneToModify {boneName=dyn_Cable_Short_01}
    rotation {pitch=-10}
   }
   n17 = "Transform (Modify) Bone - Bone: dyn_Cable_Short_02"(ComponentPose=@n16, Alpha=1, Translation="0,0,0", Rotation="40,0,0") {
    rotationMode=BMM_Additive
    rotationSpace=BCS_BoneSpace
    translationMode=BMM_Additive
    translationSpace=BCS_BoneSpace
    boneToModify {boneName=dyn_Cable_Short_02}
    rotation {pitch=40}
   }
   n18 = "Transform (Modify) Bone - Bone: base_Cable_Long_01"(ComponentPose=@n17, Alpha=1, Translation="0,0,1", Rotation="20,30,-10") {
    rotationMode=BMM_Additive
    rotationSpace=BCS_BoneSpace
    translationMode=BMM_Additive
    translationSpace=BCS_BoneSpace
    boneToModify {boneName=base_Cable_Long_01}
    rotation=<rot 20 30 -10>
    translation {z=1}
   }
   n19 = "Transform (Modify) Bone - Bone: dyn_Cable_Long_01"!(ComponentPose=@n18, Alpha=1, Translation="0,0,0", Rotation="-10,0,0") -> (Pose) {
    rotationMode=BMM_Additive
    rotationSpace=BCS_BoneSpace
    translationMode=BMM_Additive
    translationSpace=BCS_BoneSpace
    boneToModify {boneName=dyn_Cable_Long_01}
    rotation {pitch=-10}
   }
   n20 = "Transform (Modify) Bone - Bone: dyn_Cable_Long_02"(ComponentPose=@n19, Alpha=1, Translation="0,0,0", Rotation="40,0,0") {
    rotationMode=BMM_Additive
    rotationSpace=BCS_BoneSpace
    translationMode=BMM_Additive
    translationSpace=BCS_BoneSpace
    boneToModify {boneName=dyn_Cable_Long_02}
    rotation {pitch=40}
   }
   n21 = Component To Local(ComponentPose=@n20)
   n22 = Get bIsInFrontEnd()
   n23 = RigidBody!(
    ComponentPose=Local To Component(LocalPose=Use cached pose 'GripDynamics_Cache'())
    bAlphaBoolEnabled:bool=@n22
    OverrideWorldGravity:Vector="0,0,0"
    ExternalForce
   ) -> (Pose) {
    alphaInputType=Bool
    overridePhysicsAsset=&/SpacePlunge_Axe_Physics
    simulationSpace=BaseBoneSpace
    baseBoneRef {boneName=Melee_TwoHanded_Handle}
    simSpaceSettings {
     maxAngularAcceleration=4000
     maxAngularVelocity=2000
     maxLinearAcceleration=2500
     maxLinearVelocity=1200
     worldAlpha=1
     externalLinearDragV=<1 1 1>
    }
   }
   n24 = RigidBody(ComponentPose=@n23, bAlphaBoolEnabled=NOT Boolean(A=@n22), OverrideWorldGravity="0,0,0") {
    alphaInputType=Bool
    overridePhysicsAsset=&SpacePlunge_Axe_Physics
    simulationSpace=BaseBoneSpace
    baseBoneRef {boneName=Melee_TwoHanded_Handle}
    simSpaceSettings {
     maxAngularAcceleration=1600
     maxAngularVelocity=4
     maxLinearAcceleration=1000
     maxLinearVelocity=400
     worldAlpha=1
     externalLinearDragV=<1.2 1.2 1.2>
    }
   }
   Save cached pose 'DynamicsCache'(Pose=Blend Poses by bool(BlendPose_0=Component To Local(ComponentPose=@n24), BlendPose_1=Blend Poses by bool(BlendPose_0=@n9, BlendPose_1=@n21))) {cacheName=DynamicsCache}
   n25 = Anim Dynamics - Bone: dyn_Handle_FlexPoint!(ComponentPose=Local To Component(LocalPose=Use cached pose 'CRCache_Input'()), Alpha=1, GravityOverride:Vector="0,0,-980") -> (Pose) {
    angularDampingOverride=1
    angularSpringConstant=20000
    bAngularSpring=true
    bOverrideAngularDamping=true
    bUseGravityOverride=true
    gravityScale=0.35
    simulationSpace=RootRelative
    boundBone {boneName=dyn_Handle_FlexPoint}
    componentAppliedLinearAccClamp=<5000 5000 5000>
    componentLinearAccScale=<1 1 1>
    componentLinearVelScale=<1 1 1>
    gravityOverride {z=-980}
    physicsBodyDefinitions [{
      boundBone {boneName=dyn_Handle_FlexPoint}
      boxExtents=<25 35 35>
      constraintSetup {angularLimitsMax=<0 15 0>, angularLimitsMin=<0 -15 0>, angularTarget=<1 0 0>}
      localJointOffset=<2 0 0>
     }]
    relativeSpaceBone {boneName=b_AR_Root}
    simSpaceSettings {maxAngularAcceleration=12, maxAngularVelocity=3, simSpaceAngularAlpha=0.5}
   }
   n26 = Anim Dynamics - Bone: dyn_SawBase!(ComponentPose=@n25, Alpha=1, GravityOverride:Vector="0,0,-980") -> (Pose) {
    angularDampingOverride=1
    angularSpringConstant=20000
    bAngularSpring=true
    bOverrideAngularDamping=true
    bUseGravityOverride=true
    gravityScale=0.35
    simulationSpace=RootRelative
    boundBone {boneName=dyn_SawBase}
    componentAppliedLinearAccClamp=<5000 5000 5000>
    componentLinearAccScale=<1 1 1>
    componentLinearVelScale=<1 1 1>
    gravityOverride {z=-980}
    physicsBodyDefinitions [{
      boundBone {boneName=dyn_SawBase}
      boxExtents=<40 60 60>
      constraintSetup {angularLimitsMax=<0 3 0>, angularLimitsMin=<0 -10 0>, angularTarget=<1 0 0>}
      localJointOffset=<1 0 0>
     }]
    simSpaceSettings {maxAngularAcceleration=10, maxAngularVelocity=4, simSpaceAngularAlpha=1}
   }
   Save cached pose 'GripDynamics_Cache'(Pose=Component To Local(ComponentPose=@n26)) {cacheName=GripDynamics_Cache}
  ]}
 EventGraph {nodes [
   Event BlueprintUpdateAnimation()
   > n1 = Sequence()
   > n2 = "Gravity Override State Function (Parameter Asset) (Nativized)"(
    CharacterPartAnimInstance=Self-Reference()
    GameTimeInSeconds=GetGameTimeInSeconds()
    Parameters=&/SpacePlunge_Axe_Parameters
   )
   > Set GravityOverride(GravityOverride=@n2)
   > Branch(Condition=Get bIsInFrontEnd())
   > n3 = Branch(Condition=Get bIsPreviewingItemInFrontEnd() -> (bIsPreviewingItemInFrontEnd:bool))
   > Set GripDynAlpha(GripDynAlpha:double=0.2) -> (Output_Get:double)
   > Set SawDynAlpha(SawDynAlpha:double=0.5) -> (Output_Get:double)
   > Set FEOrientOffsets(FEOrientOffsets:double=0) -> (Output_Get:double)
   Set GripDynAlpha(execute=@n3.else, GripDynAlpha=1)
   > Set SawDynAlpha(SawDynAlpha=1)
   > Set FEOrientOffsets(FEOrientOffsets=1)
   n4 = Get bIsPlayingEmote()
   n5 = Get bIsPlayingMeleeAnim()
   Set isIDLE(execute=@n1.then_1, isIDLE:bool=NOT Boolean(A=OR Boolean!(A=false, B=@n4, C:bool=@n5) -> (ReturnValue))) -> (Output_Get:bool)
   TryGetPawnOwner()
   Get bIsMoving2D()
  ]}
]
variables [
 FEOrientOffsets {type=number}
 Gravity Override Params {
  type=FortGravityOverrideParameterSet
  default=(Skydive_GoingUp=(GravityOverride=(WindFrequency=(X=1.500000,Y=7.000000,Z=8.000000),WindAmplitude=(X=200.000000,Y=-500.000000,Z=400.000000),GravityOverride=(X=0.000000,Y=-300.000000,Z=1500.000000),JointName="root")),Skydive_FloatingDown=(GravityOverride=(WindFrequency=(X=1.500000,Y=7.000000,Z=8.000000),WindAmplitude=(X=200.000000,Y=-500.000000,Z=400.000000),WindOffset=(X=0.000000,Y=-600.000000,Z=400.000000),GravityOverride=(X=0.000000,Y=-700.000000,Z=600.000000),JointName="root")),Skydive_DivingDown=(GravityOverride=(WindFrequency=(X=1.500000,Y=12.000000,Z=10.000000),WindAmplitude=(X=50.000000,Y=300.000000,Z=100.000000),GravityOverride=(X=-1500.000000,Y=0.000000,Z=0.000000),JointName="pelvis")),Skydive_Parachuting=(GravityOverride=(WindFrequency=(X=1.500000,Y=7.000000,Z=8.000000),WindAmplitude=(X=200.000000,Y=-500.000000,Z=400.000000),GravityOverride=(X=0.000000,Y=-800.000000,Z=-980.000000),JointName="root")),OnGround_Standing=(GravityOverride=(GravityOverride=(X=0.000000,Y=0.000000,Z=-980.000000),JointName="root")),OnGround_Moving=(GravityOverride=(GravityOverride=(X=0.000000,Y=0.000000,Z=-980.000000),JointName="root")),OnGround_Moving_HighRotationRate=(GravityOverride=(GravityOverride=(X=0.000000,Y=0.000000,Z=-980.000000),JointName="root")),OnGround_Crouch_Moving=(GravityOverride=(GravityOverride=(X=0.000000,Y=0.000000,Z=-980.000000),JointName="root")),Emote_or_Melee=(GravityOverride=(GravityOverride=(X=0.000000,Y=0.000000,Z=-980.000000),JointName="root")),Swimming=(GravityOverride=(GravityOverride=(X=-800.000000,Y=0.000000,Z=0.000000),JointName="pelvis")),Front_End=(GravityOverride=(GravityOverride=(X=0.000000,Y=0.000000,Z=-980.000000),JointName="root")),DBNO=(GravityOverride=(GravityOverride=(X=-980.000000,Y=400.000000,Z=0.000000),JointName="pelvis")),DBNO_Carried=(GravityOverride=(GravityOverride=(X=-980.000000,Y=500.000000,Z=-500.000000),JointName="pelvis")),OnGround_Falling=(GravityOverride=(GravityOverride=(X=0.000000,Y=0.000000,Z=-980.000000),JointName="root")),Ziplining=(GravityOverride=(GravityOverride=(X=0.000000,Y=0.000000,Z=-980.000000),JointName="root")),Skydive_Surfing=(GravityOverride=(WindFrequency=(X=1.500000,Y=7.000000,Z=8.000000),WindAmplitude=(X=200.000000,Y=-500.000000,Z=400.000000),GravityOverride=(X=500.000000,Y=-300.000000,Z=-980.000000),JointName="root")),Riding_Standing=(GravityOverride=(GravityOverride=(X=0.000000,Y=-350.000000,Z=-980.000000),JointName="root")),Riding_Moving=(GravityOverride=(GravityOverride=(X=0.000000,Y=-450.000000,Z=-450.000000),JointName="root")),Riding_Moving_Targeting=(GravityOverride=(GravityOverride=(X=0.000000,Y=-200.000000,Z=-980.000000),JointName="root")),Tactical_Sprint=(GravityOverride=(GravityOverride=(X=0.000000,Y=-450.000000,Z=-500.000000),JointName="root")),Rail_Grinding=(GravityOverride=(GravityOverride=(X=0.000000,Y=0.000000,Z=-980.000000),JointName="root")),)
 }
 GravityOverride {type=Vector, default=<0 0 -980>}
 GripDynAlpha {type=number, default=1}
 isIDLE {type=boolean}
 RBAN Params {
  type=FortRigidBodyAnimNodeParameterSet
  default=(Skydive_GoingUp=(RigidBodyAnimNode=(SimSpaceSettings=(WorldAlpha=1.000000,MaxLinearVelocity=400.000000,MaxAngularVelocity=4.000000,MaxLinearAcceleration=1000.000000,MaxAngularAcceleration=1600.000000,ExternalLinearDragV=(X=1.200000,Y=1.200000,Z=1.200000)))),Skydive_FloatingDown=(RigidBodyAnimNode=(SimSpaceSettings=(WorldAlpha=1.000000,MaxLinearVelocity=400.000000,MaxAngularVelocity=4.000000,MaxLinearAcceleration=1000.000000,MaxAngularAcceleration=1600.000000,ExternalLinearDragV=(X=1.200000,Y=1.200000,Z=1.200000)))),Skydive_DivingDown=(RigidBodyAnimNode=(SimSpaceSettings=(WorldAlpha=1.000000,MaxLinearVelocity=400.000000,MaxAngularVelocity=4.000000,MaxLinearAcceleration=1000.000000,MaxAngularAcceleration=1600.000000,ExternalLinearDragV=(X=1.200000,Y=1.200000,Z=1.200000)))),Skydive_Parachuting=(RigidBodyAnimNode=(SimSpaceSettings=(WorldAlpha=1.000000,MaxLinearVelocity=400.000000,MaxAngularVelocity=4.000000,MaxLinearAcceleration=1000.000000,MaxAngularAcceleration=1600.000000,ExternalLinearDragV=(X=1.200000,Y=1.200000,Z=1.200000)))),OnGround_Standing=(RigidBodyAnimNode=(SimSpaceSettings=(WorldAlpha=1.000000,MaxLinearVelocity=400.000000,MaxAngularVelocity=4.000000,MaxLinearAcceleration=1000.000000,MaxAngularAcceleration=1600.000000,ExternalLinearDragV=(X=1.200000,Y=1.200000,Z=1.200000)))),OnGround_Moving=(RigidBodyAnimNode=(SimSpaceSettings=(WorldAlpha=1.000000,MaxLinearVelocity=400.000000,MaxAngularVelocity=4.000000,MaxLinearAcceleration=1000.000000,MaxAngularAcceleration=1600.000000,ExternalLinearDragV=(X=1.200000,Y=1.200000,Z=1.200000)))),OnGround_Moving_HighRotationRate=(RigidBodyAnimNode=(SimSpaceSettings=(WorldAlpha=1.000000,MaxLinearVelocity=400.000000,MaxAngularVelocity=4.000000,MaxLinearAcceleration=1000.000000,MaxAngularAcceleration=1600.000000,ExternalLinearDragV=(X=1.200000,Y=1.200000,Z=1.200000)))),OnGround_Crouch_Moving=(RigidBodyAnimNode=(SimSpaceSettings=(WorldAlpha=1.000000,MaxLinearVelocity=400.000000,MaxAngularVelocity=4.000000,MaxLinearAcceleration=1000.000000,MaxAngularAcceleration=1600.000000,ExternalLinearDragV=(X=1.200000,Y=1.200000,Z=1.200000)))),Emote_or_Melee=(RigidBodyAnimNode=(SimSpaceSettings=(WorldAlpha=1.000000,MaxLinearVelocity=400.000000,MaxAngularVelocity=4.000000,MaxLinearAcceleration=1000.000000,MaxAngularAcceleration=1600.000000,ExternalLinearDragV=(X=1.200000,Y=1.200000,Z=1.200000)))),Swimming=(RigidBodyAnimNode=(SimSpaceSettings=(WorldAlpha=1.000000,MaxLinearVelocity=400.000000,MaxAngularVelocity=4.000000,MaxLinearAcceleration=1000.000000,MaxAngularAcceleration=1600.000000,ExternalLinearDragV=(X=1.200000,Y=1.200000,Z=1.200000)))),Front_End=(RigidBodyAnimNode=(SimSpaceSettings=(WorldAlpha=1.000000,MaxLinearVelocity=1200.000000,MaxAngularVelocity=2000.000000,MaxLinearAcceleration=2500.000000,MaxAngularAcceleration=4000.000000,ExternalLinearDragV=(X=1.000000,Y=1.000000,Z=1.000000)))),DBNO=(RigidBodyAnimNode=(SimSpaceSettings=(WorldAlpha=1.000000,MaxLinearVelocity=400.000000,MaxAngularVelocity=4.000000,MaxLinearAcceleration=1000.000000,MaxAngularAcceleration=1600.000000,ExternalLinearDragV=(X=1.200000,Y=1.200000,Z=1.200000)))),DBNO_Carried=(RigidBodyAnimNode=(SimSpaceSettings=(WorldAlpha=1.000000,MaxLinearVelocity=400.000000,MaxAngularVelocity=4.000000,MaxLinearAcceleration=1000.000000,MaxAngularAcceleration=1600.000000,ExternalLinearDragV=(X=1.200000,Y=1.200000,Z=1.200000)))),OnGround_Falling=(RigidBodyAnimNode=(SimSpaceSettings=(WorldAlpha=1.000000,MaxLinearVelocity=400.000000,MaxAngularVelocity=4.000000,MaxLinearAcceleration=1000.000000,MaxAngularAcceleration=1600.000000,ExternalLinearDragV=(X=1.200000,Y=1.200000,Z=1.200000)))),Ziplining=(RigidBodyAnimNode=(SimSpaceSettings=(WorldAlpha=1.000000,MaxLinearVelocity=400.000000,MaxAngularVelocity=4.000000,MaxLinearAcceleration=1000.000000,MaxAngularAcceleration=1600.000000,ExternalLinearDragV=(X=1.200000,Y=1.200000,Z=1.200000)))),Skydive_Surfing=(RigidBodyAnimNode=(SimSpaceSettings=(WorldAlpha=1.000000,MaxLinearVelocity=400.000000,MaxAngularVelocity=4.000000,MaxLinearAcceleration=1000.000000,MaxAngularAcceleration=1600.000000,ExternalLinearDragV=(X=1.200000,Y=1.200000,Z=1.200000)))),Riding_Standing=(RigidBodyAnimNode=(SimSpaceSettings=(WorldAlpha=1.000000,MaxLinearVelocity=400.000000,MaxAngularVelocity=4.000000,MaxLinearAcceleration=1000.000000,MaxAngularAcceleration=1600.000000,ExternalLinearDragV=(X=1.200000,Y=1.200000,Z=1.200000)))),Riding_Moving=(RigidBodyAnimNode=(SimSpaceSettings=(WorldAlpha=1.000000,MaxLinearVelocity=400.000000,MaxAngularVelocity=4.000000,MaxLinearAcceleration=1000.000000,MaxAngularAcceleration=1600.000000,ExternalLinearDragV=(X=1.200000,Y=1.200000,Z=1.200000)))),Riding_Moving_Targeting=(RigidBodyAnimNode=(SimSpaceSettings=(WorldAlpha=1.000000,MaxLinearVelocity=400.000000,MaxAngularVelocity=4.000000,MaxLinearAcceleration=1000.000000,MaxAngularAcceleration=1600.000000,ExternalLinearDragV=(X=1.200000,Y=1.200000,Z=1.200000)))),Tactical_Sprint=(RigidBodyAnimNode=(SimSpaceSettings=(WorldAlpha=1.000000,MaxLinearVelocity=400.000000,MaxAngularVelocity=4.000000,MaxLinearAcceleration=1000.000000,MaxAngularAcceleration=1600.000000,ExternalLinearDragV=(X=1.200000,Y=1.200000,Z=1.200000)))),Rail_Grinding=(RigidBodyAnimNode=(SimSpaceSettings=(WorldAlpha=1.000000,MaxLinearVelocity=400.000000,MaxAngularVelocity=4.000000,MaxLinearAcceleration=1000.000000,MaxAngularAcceleration=1600.000000,ExternalLinearDragV=(X=1.200000,Y=1.200000,Z=1.200000)))),)
 }
 SawDynAlpha {type=number, default=1}
]
